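# Translucent, unlit editor utility material, named for previewing image reflections.
# It multiplies the RGB of a Texture parameter, defaulting to DefaultDiffuse, by the RGB of a Color vector parameter.
# Both inputs are exposed parameters, so instances can swap the image and tint it.
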Material /Engine/EditorMaterials/Utilities/ImageReflectionPreview
graph {
 properties {blendMode=BLEND_Translucent, pixelDepthOffsetMode=PDOM_Legacy, shadingModel=MSM_Unlit, usedShadingModels=Unlit}
 nodes [Multiply(
   A=TextureSampleParameter2D() {parameterName=Texture, texture=/Engine/EngineMaterials/DefaultDiffuse}.RGB
   B=VectorParameter() {parameterName=Color}.RGB
  )]
}
