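# LinearColorPicker_MAT is an editor utility material, named for a linear color picker, using the Default Lit shading model.
# A Color vector parameter, white by default, feeds its RGB into the DiffuseColor input of a ConvertFromDiffSpec function call.
# The graph shows no connection from that function to any material output, so the parameter has no visible effect on shading.
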Material /Engine/EditorMaterials/Utilities/LinearColorPicker_MAT
graph {
 properties {pixelDepthOffsetMode=PDOM_Legacy, usedShadingModels=Default Lit}
 nodes [MaterialFunctionCall(DiffuseColor=VectorParameter() {desc=Color, parameterName=Color, defaultValue=<rgba 1 1 1 1>}.RGB) {materialFunction=&ConvertFromDiffSpec}]
}
